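# Niagara system for the BuffCut emote's hovering prop, layering fog, bokeh sparkle and smoke-ring streak effects.
# Active "001" emitters sit beside disabled originals, and the active fog copy samples a shared skeletal mesh reader for spawn positions.
# Spawn groups split emitters into sub-effects like rising versus ground fog, or smoke, streaks and ring.
# Time-of-day helpers modulate color and opacity, while owner speed and front-end modifiers set fog velocity values.
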NiagaraSystem /CosmeticCompanions/FX/Builtins/BuffCut/Emote/NS_Emote_BuffCut_Prop_Hover
dependencies {
 usedDataInterfaces [&NiagaraDataInterfaceArrayColor, &NiagaraDataInterfaceCurve, &NiagaraDataInterfaceSkeletalMesh]
 usedDynamicInputs [
  &MaskVector2dBySpawnGroup
  &MaskVectorBySpawnGroup
  &RotateVectorByQuaternion
  &MaskFloat
 ]
 usedModules [
  &self:SystemSpawnScript.SystemScriptSource.SystemScriptGraph.NiagaraNodeAssignment_0.SetVariables_#1
  &SystemState
  &Timeline
 ]
}
emitters {
 Fog_system_Condensed {
  bEnabled=false
  emitterUpdateScript [
   &EmitterState(Life Cycle Mode=::NewEnumerator1, Loop Behavior=::NewEnumerator1, Loop Duration=10, UseLoopDelay=false)
   &SpawnBurst_Instantaneous(Spawn Group=1, Use Spawn Probability=false, UseLoopCountLimit=false)
   &SpawnRate(Spawn Probability=0.5, SpawnRate=10, Use Spawn Probability=true)
  ]
  particleSpawnScript [
   /SetVariables(Particles.VisibilityTag=&SpawnGroup()) {isSetParameters=true, module=SetVariables_#2}
   &InitializeParticle(
    Lifetime=&MergeHelper_FloatSwitch(
     0=&RandomRangeFloat(Minimum=0.6, Maximum=1)
     1=10
     2=?
     3=?
     Index=&SpawnGroup()
     Num=::NewEnumerator0
    )
    Position Offset=<0 0 35>
    Sprite Rotation Mode=::NewEnumerator2
    Sprite Size=&MergeHelper_Vector2Switch(
     0=&Vector2DFromFloat(Value=&RandomRangeFloat(Minimum=25, Maximum=35))
     1=&RandomRangeVector2D(Minimum=<20 10>, Maximum=<50 35>)
     2=?
     3=?
     4=?
     Index=&SpawnGroup()
     Num=::NewEnumerator0
    )
    Sprite Size Mode=::NewEnumerator4
    UsePositionOffset=true
    Write Material Random=true
   )
   /SetVariables(Particles.0 - Env Fog Velocity:Vector3f=&MakeVector(
     X=0
     Y=&FN_FrontEndFloatModifier(
      Gameplay=&MakeCustomFloatFromBool(
       Bool=&SetBoolByFloatComparison(A=&VectorLength(Vector=$User.OwnerLocalVelocity), B=90)
       True Float=&RandomRangeFloat(Minimum=-20, Maximum=-50)
       False Float=&RandomRangeFloat(Minimum=-10, Maximum=-15)
      )
      Front End=7
     )
     Z=0
    )) {isSetParameters=true, module=SetVariables_#3}
   /SetVariables(Particles.1 - Ground Fog Velocity:Vector3f=&MakeVector(
     X=0
     Y=&FN_FrontEndFloatModifier(
      Gameplay=&MakeCustomFloatFromBool(
       Bool=&SetBoolByFloatComparison(A=&VectorLength(Vector=$User.OwnerLocalVelocity), B=90)
       True Float=-80
       False Float=-80
      )
      Front End=-80
     )
     Z=0
    )) {isSetParameters=true, module=SetVariables_#4}
   &ShapeLocation(
    MaskSpawnGroup=false
    Non Uniform Scale=&MergeHelper_VectorSwitch(0=<1 1 0.5>, 1=<1 1 0.1>, 2=?, 3=?, Index=&SpawnGroup(), Num=::NewEnumerator0)
    Override Seed=false
    Sphere Radius=30
    UseOwnerScale=true
   )
   &AddVelocity(Velocity=&MakeVector(
     X=0
     Y=0
     Z=&MergeHelper_FloatSwitch(
      0=&RandomRangeFloat(Minimum=20, Maximum=75)
      1=0
      2=?
      3=?
      Index=&SpawnGroup()
      Num=::NewEnumerator0
     )
    ))
   &CameraOffset(Camera Offset Amount=&RandomRangeFloat(Minimum=0, Maximum=50))
  ]
  particleUpdateScript [
   &ParticleState()
   &ScaleSpriteSize(
    Uniform Curve Scale=10
    Uniform Curve Sprite Scale=DataInterface() {propertyValues {Curve {defaultValue=1, keys [{tangentMode=RCTM_User, time=0, value=0.7}, {tangentMode=RCTM_User, time=1, value=1}]}}}
   )
   &SubUVAnimation(
    End Frame=15
    SubUV Animation Mode:ENiagara_SubUVLookupMode=::NewEnumerator1
    SubUV Curve=DataInterface() {propertyValues {Curve {keys [{tangentMode=RCTM_User, time=0, value=0}, {tangentMode=RCTM_User, time=1.0138, value=0.388}]}}}
   )
   &Drag(
    Drag=&Multiply_Float(
     A=&RandomRangeFloat(Minimum=0.5, Maximum=1)
     B=&FloatFromCurve(FloatCurve=DataInterface() {propertyValues {Curve {keys [
          {interpMode=RCIM_Cubic, leaveTangent=0.248, tangentMode=RCTM_Break, time=0, value=0}
          {interpMode=RCIM_Cubic, tangentMode=RCTM_Break, time=1, value=1}
         ]}}})
    )
    Use Linear Drag=true
    Use Rotational Drag=false
   )
   &SolveForcesAndVelocity(Clamp Velocity=false, Limit Acceleration=false)
   /SetVariables(
    Particles.0 - Env Fog Scale Color:LinearColor=&MakeLinearColorFromVectorAndFloat(
     "Vector (RGB)"=<1 1 1>
     "Float (Alpha)"=&FloatFromCurve(
      FloatCurve=DataInterface() {propertyValues {Curve {keys [
          {arriveTangent=3.5962, interpMode=RCIM_Cubic, leaveTangent=3.5962, tangentMode=RCTM_User, time=0, value=0}
          {
           arriveTangent=0.1817
           arriveTangentWeight=0.1842
           interpMode=RCIM_Cubic
           leaveTangent=0.1817
           leaveTangentWeight=0.7369
           tangentMode=RCTM_User
           time=0.5037
           value=0.7647
          }
          {arriveTangent=-4.7233, tangentMode=RCTM_User, time=1, value=0}
         ]}}}
      Scale Curve=&RandomRangeFloat(
       Minimum=0.02
       Maximum=&FNTimeOfDayScalar(Night Scale=0.05, Morning Scale=0.05, Day Scale=0.07, Evening Scale=0.05)
      )
     )
    )
    Particles.1 - Ground Fog Scale Color:LinearColor=&MakeLinearColorFromVectorAndFloat(
     "Vector (RGB)"=<1 1 1>
     "Float (Alpha)"=&FloatFromCurve(
      FloatCurve=DataInterface() {propertyValues {Curve {keys [
          {arriveTangent=5, interpMode=RCIM_Cubic, time=0, value=0}
          {arriveTangent=1.5181, interpMode=RCIM_Cubic, leaveTangent=1.5181, time=0.2, value=0.7}
          {
           arriveTangent=0.1817
           arriveTangentWeight=0.1842
           interpMode=RCIM_Cubic
           leaveTangent=0.1817
           leaveTangentWeight=0.7369
           tangentMode=RCTM_User
           time=0.5037
           value=0.7647
          }
          <at 1 0>
         ]}}}
      Scale Curve=&RandomRangeFloat(
       Minimum=&FNTimeOfDayScalar(Night Scale=0.035, Morning Scale=0.07, Day Scale=0.05, Evening Scale=0.07)
       Maximum=&FNTimeOfDayScalar(Night Scale=0.05, Morning Scale=0.07, Day Scale=0.07, Evening Scale=0.1)
      )
     )
    )
   ) {isSetParameters=true, module=SetVariables_#5}
   /SetVariables(
    Particles.0 - Env Fog Dyn Param:Vector4f=&MakeVector4(X=0, Y=1, Z=1, W=1)
    Particles.1 - Ground Fog Dyn Param:Vector4f=&MakeVector4(X=0, Y=1, Z=0.1, W=1)
   ) {isSetParameters=true, module=SetVariables_#6}
   &BlendSpace(Rotation Contribution=0.5, Translation Contribution=0.95) {enabled=false}
   &ScaleColor(
    Scale Alpha=&FloatFromCurve(
     FloatCurve=DataInterface() {propertyValues {Curve {keys [
         {arriveTangent=3.5962, interpMode=RCIM_Cubic, leaveTangent=3.5962, tangentMode=RCTM_User, time=0, value=0}
         {
          arriveTangent=0.1817
          arriveTangentWeight=0.1842
          interpMode=RCIM_Cubic
          leaveTangent=0.1817
          leaveTangentWeight=0.7369
          tangentMode=RCTM_User
          time=0.3249
          value=0.6411
         }
         {arriveTangent=-1.3091, tangentMode=RCTM_User, time=1, value=0.0073}
        ]}}}
     Scale Curve=&RandomRangeFloat(
      Minimum=0.02
      Maximum=&FNTimeOfDayScalar(Night Scale=0.03, Morning Scale=0.03, Day Scale=0.05, Evening Scale=0.03)
     )
    )
    ScaleA=true
    ScaleRGB=true
   )
   &DynamicMaterialParameters(
    Index 0 Param 1=50
    Module.Index0Float0Write=true
    Module.Index0Float1Write=true
    Module.Index0Float2Write=true
    Module.Index0Float3Write=true
   )
   &FN_Light_Attributes(
    Color=<rgba 3.63 4.1486 5 1>
    Create Color Attribute=true
    Create Exponent Attribute=true
    Create Position Attribute=true
    Create Radius Attribute=true
    Enable Particle Light Enable=true
    Enable Visibility Tag=true
    Exponent:NiagaraFloat=&FN_FrontEndFloatModifier(Gameplay=35, Front End=8)
    Light Enabled=true
    Position=&AddVectorToPosition(Position=$Engine.Emitter.SimulationPosition, Vector=<5 5 40>)
    Radius=40
    Scale Color=&Multiply_Float(
     A=0.5
     B=&Multiply_Float(
      A=&Waveform("Waveform [1]"=::NewEnumerator0, "[1] Frequency"=3, "[1] Amplitude Min/Max"=<0.7 1>, "[1] Scale"=2.6368)
      B=&FNTimeOfDayScalar(Night Scale=0.5, Morning Scale=0.4, Day Scale=0.45, Evening Scale=0.6)
     )
    )
    VisibilityTag=1
   ) {enabled=false, module=FN_Light_Attributes001}
  ]
  rendererClasses [
   &NiagaraSpriteRendererProperties
   &NiagaraLightRendererProperties
   &NiagaraMeshRendererProperties
  ]
  renderers [
   {rendererClass=&NiagaraSpriteRendererProperties}
   {rendererClass=&NiagaraLightRendererProperties, rendererIndex=1}
   {rendererClass=&NiagaraMeshRendererProperties, rendererIndex=2}
  ]
 }
 Fog_system_Condensed001 {
  emitterUpdateScript [
   &EmitterState(Life Cycle Mode=::NewEnumerator1, Loop Behavior=::NewEnumerator1, Loop Duration=10, UseLoopDelay=false)
   &SpawnBurst_Instantaneous(Spawn Group=1, Use Spawn Probability=false, UseLoopCountLimit=false)
   &SpawnRate(Spawn Probability=0.5, SpawnRate=15, Use Spawn Probability=true)
  ]
  particleSpawnScript [
   /SetVariables(Particles.VisibilityTag=&SpawnGroup()) {isSetParameters=true, module=SetVariables_#2}
   &InitializeParticle(
    Color=&FNTimeOfDayColor(
     Night Color=<rgba 0.4115 0.4861 0.7948 1>
     Morning Color=<rgba 1 1 1 1>
     Day Color=<rgba 1 1 1 1>
     Evening Color=<rgba 0.6251 0.6573 0.6508 1>
    )
    Lifetime=&MergeHelper_FloatSwitch(
     0=&RandomRangeFloat(Minimum=0.6, Maximum=1)
     1=10
     2=?
     3=?
     Index=&SpawnGroup()
     Num=::NewEnumerator0
    )
    Position Offset=<0 -25 35>
    Sprite Rotation Mode=::NewEnumerator2
    Sprite Size=&MergeHelper_Vector2Switch(
     0=&Vector2DFromFloat(Value=&RandomRangeFloat(Minimum=25, Maximum=35))
     1=&RandomRangeVector2D(Minimum=<20 10>, Maximum=<50 35>)
     2=?
     3=?
     4=?
     Index=&SpawnGroup()
     Num=::NewEnumerator0
    )
    Sprite Size Mode=::NewEnumerator4
    UsePositionOffset=false
    Write Material Random=true
   )
   /SetVariables(Particles.0 - Env Fog Velocity:Vector3f=&MakeVector(
     X=0
     Y=&FN_FrontEndFloatModifier(
      Gameplay=&MakeCustomFloatFromBool(
       Bool=&SetBoolByFloatComparison(A=&VectorLength(Vector=$User.OwnerLocalVelocity), B=90)
       True Float=&RandomRangeFloat(Minimum=-20, Maximum=-50)
       False Float=&RandomRangeFloat(Minimum=-10, Maximum=-15)
      )
      Front End=7
     )
     Z=0
    )) {isSetParameters=true, module=SetVariables_#3}
   /SetVariables(Particles.1 - Ground Fog Velocity:Vector3f=&MakeVector(
     X=0
     Y=&FN_FrontEndFloatModifier(
      Gameplay=&MakeCustomFloatFromBool(
       Bool=&SetBoolByFloatComparison(A=&VectorLength(Vector=$User.OwnerLocalVelocity), B=90)
       True Float=-80
       False Float=-80
      )
      Front End=-80
     )
     Z=0
    )) {isSetParameters=true, module=SetVariables_#4}
   &SkeletalMeshLocation(
    Bone Sampling Mode=::NewEnumerator4
    MaskSpawnGroup=false
    Position Sampling=::NewEnumerator1
    Sampled Position Offset=<0 -45 0>
    Sampled Position Offset Coordinate Space:ENiagaraSkelMeshTransforms=::NewEnumerator2
    Skeletal Mesh=$System.Skeletal Mesh Reader
   )
   &ShapeLocation(
    MaskSpawnGroup=false
    Offset=<0 -10 -15>
    Offset Coordinate Space=::Local
    Offset Mode=::NewEnumerator0
    Sphere Distribution=::NewEnumerator3
    Sphere Radius=35
    UseOwnerScale=true
   )
   &AddVelocity(Velocity=&MakeVector(
     X=0
     Y=0
     Z=&MergeHelper_FloatSwitch(
      0=&RandomRangeFloat(Minimum=20, Maximum=75)
      1=0
      2=?
      3=?
      Index=&SpawnGroup()
      Num=::NewEnumerator0
     )
    ))
   &CameraOffset(Camera Offset Amount=&RandomRangeFloat(Minimum=0, Maximum=50))
  ]
  particleUpdateScript [
   &ParticleState()
   &ScaleSpriteSize(
    Uniform Curve Scale=10
    Uniform Curve Sprite Scale=DataInterface() {propertyValues {Curve {defaultValue=1, keys [{tangentMode=RCTM_User, time=0, value=0.7}, {tangentMode=RCTM_User, time=1, value=1}]}}}
   )
   &SubUVAnimation(
    End Frame=15
    SubUV Animation Mode:ENiagara_SubUVLookupMode=::NewEnumerator1
    SubUV Curve=DataInterface() {propertyValues {Curve {keys [{tangentMode=RCTM_User, time=0, value=0}, {tangentMode=RCTM_User, time=1.0138, value=0.388}]}}}
   )
   &Drag(
    Drag=&Multiply_Float(
     A=&RandomRangeFloat(Minimum=0.5, Maximum=1)
     B=&FloatFromCurve(FloatCurve=DataInterface() {propertyValues {Curve {keys [
          {interpMode=RCIM_Cubic, leaveTangent=0.248, tangentMode=RCTM_Break, time=0, value=0}
          {interpMode=RCIM_Cubic, tangentMode=RCTM_Break, time=1, value=1}
         ]}}})
    )
    Use Linear Drag=true
    Use Rotational Drag=false
   )
   &SolveForcesAndVelocity(Clamp Velocity=false, Limit Acceleration=false)
   /SetVariables(
    Particles.0 - Env Fog Scale Color:LinearColor=&MakeLinearColorFromVectorAndFloat(
     "Vector (RGB)"=<1 1 1>
     "Float (Alpha)"=&FloatFromCurve(
      FloatCurve=DataInterface() {propertyValues {Curve {keys [
          {arriveTangent=3.5962, interpMode=RCIM_Cubic, leaveTangent=3.5962, tangentMode=RCTM_User, time=0, value=0}
          {
           arriveTangent=0.1817
           arriveTangentWeight=0.1842
           interpMode=RCIM_Cubic
           leaveTangent=0.1817
           leaveTangentWeight=0.7369
           tangentMode=RCTM_User
           time=0.5037
           value=0.7647
          }
          {arriveTangent=-4.7233, tangentMode=RCTM_User, time=1, value=0}
         ]}}}
      Scale Curve=&RandomRangeFloat(
       Minimum=0.02
       Maximum=&FNTimeOfDayScalar(Night Scale=0.05, Morning Scale=0.05, Day Scale=0.07, Evening Scale=0.05)
      )
     )
    )
    Particles.1 - Ground Fog Scale Color:LinearColor=&MakeLinearColorFromVectorAndFloat(
     "Vector (RGB)"=<1 1 1>
     "Float (Alpha)"=&FloatFromCurve(
      FloatCurve=DataInterface() {propertyValues {Curve {keys [
          {arriveTangent=5, interpMode=RCIM_Cubic, time=0, value=0}
          {arriveTangent=1.5181, interpMode=RCIM_Cubic, leaveTangent=1.5181, time=0.2, value=0.7}
          {
           arriveTangent=0.1817
           arriveTangentWeight=0.1842
           interpMode=RCIM_Cubic
           leaveTangent=0.1817
           leaveTangentWeight=0.7369
           tangentMode=RCTM_User
           time=0.5037
           value=0.7647
          }
          <at 1 0>
         ]}}}
      Scale Curve=&RandomRangeFloat(
       Minimum=&FNTimeOfDayScalar(Night Scale=0.035, Morning Scale=0.07, Day Scale=0.05, Evening Scale=0.07)
       Maximum=&FNTimeOfDayScalar(Night Scale=0.05, Morning Scale=0.07, Day Scale=0.07, Evening Scale=0.1)
      )
     )
    )
   ) {isSetParameters=true, module=SetVariables_#5}
   /SetVariables(
    Particles.0 - Env Fog Dyn Param:Vector4f=&MakeVector4(X=0, Y=1, Z=1, W=1)
    Particles.1 - Ground Fog Dyn Param:Vector4f=&MakeVector4(X=0, Y=1, Z=0.1, W=1)
   ) {isSetParameters=true, module=SetVariables_#6}
   &BlendSpace(Rotation Contribution=0.5, Translation Contribution=0.95) {enabled=false}
   &ScaleColor(
    Scale Alpha=&FloatFromCurve(
     FloatCurve=DataInterface() {propertyValues {Curve {keys [
         {arriveTangent=3.5962, interpMode=RCIM_Cubic, leaveTangent=3.5962, tangentMode=RCTM_User, time=0, value=0}
         {
          arriveTangent=0.1817
          arriveTangentWeight=0.1842
          interpMode=RCIM_Cubic
          leaveTangent=0.1817
          leaveTangentWeight=0.7369
          tangentMode=RCTM_User
          time=0.3249
          value=0.6411
         }
         {arriveTangent=-1.3091, tangentMode=RCTM_User, time=1, value=0.0073}
        ]}}}
     Scale Curve=&RandomRangeFloat(
      Minimum=0.02
      Maximum=&FNTimeOfDayScalar(Night Scale=0.03, Morning Scale=0.03, Day Scale=0.05, Evening Scale=0.03)
     )
    )
    ScaleA=true
    ScaleRGB=true
   )
   &DynamicMaterialParameters(
    Index 0 Param 1=50
    Module.Index0Float0Write=true
    Module.Index0Float1Write=true
    Module.Index0Float2Write=true
    Module.Index0Float3Write=true
   )
   &FN_Light_Attributes(
    Color=<rgba 3.63 4.1486 5 1>
    Create Color Attribute=true
    Create Exponent Attribute=true
    Create Position Attribute=true
    Create Radius Attribute=true
    Enable Particle Light Enable=true
    Enable Visibility Tag=true
    Exponent:NiagaraFloat=&FN_FrontEndFloatModifier(Gameplay=35, Front End=8)
    Light Enabled=true
    Position=&AddVectorToPosition(Position=$Engine.Emitter.SimulationPosition, Vector=<5 5 40>)
    Radius=40
    Scale Color=&Multiply_Float(
     A=0.5
     B=&Multiply_Float(
      A=&Waveform("Waveform [1]"=::NewEnumerator0, "[1] Frequency"=3, "[1] Amplitude Min/Max"=<0.7 1>, "[1] Scale"=2.6368)
      B=&FNTimeOfDayScalar(Night Scale=0.5, Morning Scale=0.4, Day Scale=0.45, Evening Scale=0.6)
     )
    )
    VisibilityTag=1
   ) {enabled=false, module=FN_Light_Attributes001}
  ]
  rendererClasses [
   &NiagaraSpriteRendererProperties
   &NiagaraLightRendererProperties
   &NiagaraMeshRendererProperties
  ]
  renderers [
   {rendererClass=&NiagaraSpriteRendererProperties}
   {rendererClass=&NiagaraLightRendererProperties, rendererIndex=1}
   {rendererClass=&NiagaraMeshRendererProperties, rendererIndex=2}
  ]
 }
 Rainbow_bokeh {
  bEnabled=false
  emitterUpdateScript [
   &EmitterState()
   &SpawnRate(SpawnRate=5, Use Spawn Probability=false)
  ]
  particleSpawnScript [
   &InitializeParticle(
    Color=&SelectLinearColorFromArray(
     Array Sampling Mode=::NewEnumerator0
     Direct Array Index=?
     Direct Array Mode=::NewEnumerator1
     Color Selection Array=DataInterface() {propertyValues {ColorData [<rgba 0.4073 0.4073 0.4073 1>, <rgba 1 1 1 1>]}}
     Random Seed=0
    )
    Lifetime Max=1.8
    Lifetime Min=0.4
    Lifetime Mode=::NewEnumerator1
    Sprite Rotation Angle=&UniformAOrBFloat(A=0, B=0.125, Evaluation Type=::NewEnumerator0, Recalculate Random Each Loop=true, Override Seed=false)
    Sprite Rotation Mode=::NewEnumerator1
    Sprite Size Mode=::NewEnumerator1
    Uniform Sprite Size Max=6
    Uniform Sprite Size Min=2
    UsePositionOffset=true
    Write Material Random=true
   )
   &ShapeLocation(
    MaskSpawnGroup=false
    Non Uniform Scale=<0.7 0.7 0.8>
    Offset=<0 0 50>
    Offset Mode=::NewEnumerator0
    Override Seed=false
    Shape Origin=<0 0 0>
    Sphere Radius=50
    UseOwnerScale=true
   )
   &CameraOffset(Camera Offset Amount=-100)
  ]
  particleUpdateScript [
   &ParticleState()
   &ScaleSpriteSize(Uniform Curve Sprite Scale=DataInterface() {propertyValues {Curve {
       defaultValue=0.6687
       keys [
        {interpMode=RCIM_Cubic, tangentMode=RCTM_User, time=0, value=0}
        {arriveTangent=-0.4264, interpMode=RCIM_Cubic, leaveTangent=-0.4264, tangentMode=RCTM_User, time=0.5, value=1}
       ]
      }}}) {enabled=false}
   &ScaleColor(
    Scale Alpha=&FloatFromCurve(
     FloatCurve=DataInterface() {propertyValues {Curve {keys [
         {arriveTangent=9.8157, interpMode=RCIM_Cubic, leaveTangent=9.8157, tangentMode=RCTM_User, time=0, value=0}
         {arriveTangent=-0.0381, interpMode=RCIM_Cubic, leaveTangent=-0.0381, tangentMode=RCTM_User, time=0.1675, value=0.6}
         {interpMode=RCIM_Cubic, tangentMode=RCTM_User, time=0.3, value=0.35}
         {interpMode=RCIM_Cubic, tangentMode=RCTM_User, time=0.5, value=0.5}
         {interpMode=RCIM_Cubic, tangentMode=RCTM_Break, time=1, value=0}
        ]}}}
     Scale Curve=&FNTimeOfDayScalar(Night Scale=0.35, Morning Scale=0.35, Day Scale=0.4, Evening Scale=0.35)
    )
    ScaleA=true
    ScaleRGB=true
   )
   &Drag(Drag=2.5, Use Linear Drag=true, Use Rotational Drag=false) {enabled=false}
   &GravityForce(Gravity=&RandomRangeVector(Minimum=<-5 -5 -5>, Maximum=<5 -5 5>))
   &SolveForcesAndVelocity(Clamp Velocity=false, Limit Acceleration=false)
   /SetVariables(Particles.New Position:NiagaraPosition=&Add_Vector(
     A=Linked() {linkedVariable {name=Particles.Position, type=NiagaraPosition}}
     B=Linked() {linkedVariable {name=Engine.Owner.Position, type=NiagaraPosition}}
    )) {isSetParameters=true, module=SetVariables_#7}
  ]
  rendererClasses [&NiagaraSpriteRendererProperties]
  renderers [{rendererClass=&NiagaraSpriteRendererProperties}]
 }
 Rainbow_bokeh001 {
  emitterUpdateScript [
   &EmitterState()
   &SpawnRate(SpawnRate=5, Use Spawn Probability=false)
  ]
  particleSpawnScript [
   &InitializeParticle(
    Color=&SelectLinearColorFromArray(
     Array Sampling Mode=::NewEnumerator0
     Direct Array Index=?
     Direct Array Mode=::NewEnumerator1
     Color Selection Array=DataInterface() {propertyValues {ColorData [<rgba 0.4073 0.4073 0.4073 1>, <rgba 1 1 1 1>]}}
     Random Seed=0
    )
    Lifetime Max=1.8
    Lifetime Min=0.4
    Lifetime Mode=::NewEnumerator1
    Sprite Rotation Angle=&UniformAOrBFloat(A=0, B=0.125, Evaluation Type=::NewEnumerator0, Recalculate Random Each Loop=true, Override Seed=false)
    Sprite Rotation Mode=::NewEnumerator1
    Sprite Size Mode=::NewEnumerator1
    Uniform Sprite Size Max=6
    Uniform Sprite Size Min=2
    UsePositionOffset=true
    Write Material Random=true
   )
   &ShapeLocation(
    MaskSpawnGroup=false
    Non Uniform Scale=<0.7 0.7 0.8>
    Offset=<0 0 50>
    Offset Mode=::NewEnumerator0
    Override Seed=false
    Shape Origin=<0 0 0>
    Sphere Radius=50
    UseOwnerScale=true
   )
   &CameraOffset(Camera Offset Amount=-100)
  ]
  particleUpdateScript [
   &ParticleState()
   &ScaleSpriteSize(Uniform Curve Sprite Scale=DataInterface() {propertyValues {Curve {
       defaultValue=0.6687
       keys [
        {interpMode=RCIM_Cubic, tangentMode=RCTM_User, time=0, value=0}
        {arriveTangent=-0.4264, interpMode=RCIM_Cubic, leaveTangent=-0.4264, tangentMode=RCTM_User, time=0.5, value=1}
       ]
      }}}) {enabled=false}
   &ScaleColor(
    Scale Alpha=&FloatFromCurve(
     FloatCurve=DataInterface() {propertyValues {Curve {keys [
         {arriveTangent=9.8157, interpMode=RCIM_Cubic, leaveTangent=9.8157, tangentMode=RCTM_User, time=0, value=0}
         {arriveTangent=-0.0381, interpMode=RCIM_Cubic, leaveTangent=-0.0381, tangentMode=RCTM_User, time=0.1675, value=0.6}
         {interpMode=RCIM_Cubic, tangentMode=RCTM_User, time=0.3, value=0.35}
         {interpMode=RCIM_Cubic, tangentMode=RCTM_User, time=0.5, value=0.5}
         {interpMode=RCIM_Cubic, tangentMode=RCTM_Break, time=1, value=0}
        ]}}}
     Scale Curve=&FNTimeOfDayScalar(Night Scale=0.35, Morning Scale=0.35, Day Scale=0.4, Evening Scale=0.35)
    )
    ScaleA=true
    ScaleRGB=true
   )
   &Drag(Drag=2.5, Use Linear Drag=true, Use Rotational Drag=false) {enabled=false}
   &GravityForce(Gravity=&RandomRangeVector(Minimum=<-5 -5 -5>, Maximum=<5 -5 5>))
   &SolveForcesAndVelocity(Clamp Velocity=false, Limit Acceleration=false)
   /SetVariables(Particles.New Position:NiagaraPosition=&Add_Vector(
     A=Linked() {linkedVariable {name=Particles.Position, type=NiagaraPosition}}
     B=Linked() {linkedVariable {name=Engine.Owner.Position, type=NiagaraPosition}}
    )) {isSetParameters=true, module=SetVariables_#7}
  ]
  rendererClasses [&NiagaraSpriteRendererProperties]
  renderers [{rendererClass=&NiagaraSpriteRendererProperties}]
 }
 SmokeRingStreaks {
  bEnabled=false
  emitterUpdateScript [
   &EmitterState(
    Delay First Loop Only=true
    Life Cycle Mode=::NewEnumerator1
    Loop Delay=0.2
    Loop Duration=&RandomRangeFloat(Minimum=0.5, Maximum=0.8)
    Recalculate Duration Each Loop=true
    UseLoopDelay=true
   )
   &SpawnBurst_Instantaneous(
    Spawn Count=&RandomRangeInt(Minimum=15, Maximum=24)
    Use Spawn Probability=false
    UseLoopCountLimit=false
   )
   &SpawnBurst_Instantaneous(
    Spawn Count=&RandomRangeInt(Minimum=15, Maximum=24)
    Spawn Group=1
    Use Spawn Probability=false
    UseLoopCountLimit=false
   ) {module=SpawnBurst_Instantaneous001}
   &SpawnBurst_Instantaneous(Spawn Group=2, Use Spawn Probability=false, UseLoopCountLimit=false) {module=SpawnBurst_Instantaneous002}
  ]
  particleSpawnScript [
   /SetVariables(Particles.VisibilityTag=&SpawnGroup()) {isSetParameters=true, module=SetVariables_#8}
   &InitializeParticle(
    Color=&SelectLinearColorFromArray(
     Array Sampling Mode=::NewEnumerator0
     Direct Array Index=?
     Color Selection Array=DataInterface() {propertyValues {ColorData [<rgba 0.903 0.9271 0.1598 0.05>, <rgba 0.2392 0.9271 0.1932 0.05>, <rgba 0.1336 0.9271 0.8384 0.05>]}}
     Random Seed=0
    )
    Lifetime Max=&MaskFloatBySpawnGroup(
     Masked Float=0.8
     Passthrough Float=&MaskFloatBySpawnGroup(Masked Float=0.4, Passthrough Float=0.8, Spawn Group Mask=1)
    )
    Lifetime Min:NiagaraFloat=&MaskFloatBySpawnGroup(
     Masked Float=0.5
     Passthrough Float=&MaskFloatBySpawnGroup(Masked Float=0.35, Passthrough Float=0.5, Spawn Group Mask=1)
    )
    Lifetime Mode=::NewEnumerator1
    Sprite Rotation Mode=::NewEnumerator2
    Sprite Size Mode=::NewEnumerator3
    Uniform Sprite Size=&RandomRangeFloat(Minimum=15, Maximum=25)
    UsePositionOffset=false
    Write Material Random=true
   )
   &ShapeLocation(
    Cylinder Height=&MaskFloatBySpawnGroup(Masked Float=5, Passthrough Float=0)
    Cylinder Height Midpoint=0
    Cylinder Radius=&MaskFloatBySpawnGroup(
     Masked Float=25
     Passthrough Float=&MaskFloatBySpawnGroup(Masked Float=0.5, Passthrough Float=0, Spawn Group Mask=1)
    )
    MaskSpawnGroup=false
    Override Seed=false
    Shape Primitive=::NewEnumerator1
    Surface Only=true
    Use Endcaps In Surface Only Mode=false
    UseOwnerScale=true
   )
   &AddVelocity(Velocity=&MakeVector(
     X=0
     Y=0
     Z=&UniformRangedFloat(
      Minimum=0
      Maximum=&MaskFloatBySpawnGroup(
       Masked Float=0
       Passthrough Float=&MaskFloatBySpawnGroup(Masked Float=50, Passthrough Float=0, Spawn Group Mask=1)
      )
     )
    )) {module=AddVelocity001}
   &AddVelocity(
    Module.Use Falloff Distance=false
    Override Seed=false
    Velocity Mode=::NewEnumerator1
    Velocity Speed=&RandomRangeFloat(
     Minimum=&MaskFloatBySpawnGroup(
      Masked Float=125
      Passthrough Float=&MaskFloatBySpawnGroup(Masked Float=150, Passthrough Float=0, Spawn Group Mask=1)
     )
     Maximum=&MaskFloatBySpawnGroup(
      Masked Float=200
      Passthrough Float=&MaskFloatBySpawnGroup(Masked Float=150, Passthrough Float=0, Spawn Group Mask=1)
     )
    )
   )
   &CameraOffset(Camera Offset Amount=5)
   /SetVariables(Particles.WindStreakDP:Vector4f=&MakeVector4(
     X=&RandomRangeFloat(Minimum=0, Maximum=1)
     Y=&RandomRangeFloat(Minimum=1, Maximum=2)
     Z=1
     W=1
    )) {isSetParameters=true, module=SetVariables_#9}
   /SetVariables(Particles.RingAlignment:Vector3f=<0 1 0>, Particles.RingFacing:Vector3f=<0 0 1>) {isSetParameters=true, module=SetVariables_#10}
  ]
  particleUpdateScript [
   &ParticleState()
   &GravityForce(Gravity=&MakeVector(
     X=0
     Y=0
     Z=&FloatFromCurve(
      FloatCurve=DataInterface() {propertyValues {Curve {keys [<at 0 0 RCIM_Cubic>,<at 0.4 1 RCIM_Cubic>]}}}
      Scale Curve=&MaskFloatBySpawnGroup(Masked Float=&RandomRangeFloat(Minimum=50, Maximum=125), Passthrough Float=0)
     )
    ))
   &Drag(
    Drag=&MaskFloatBySpawnGroup(Masked Float=10, Passthrough Float=1.5)
    Use Linear Drag=true
    Use Rotational Drag=false
   )
   &SolveForcesAndVelocity(Clamp Velocity=false, Limit Acceleration=false)
   &ScaleSpriteSize(Uniform Curve Sprite Scale=DataInterface() {propertyValues {Curve {keys [
        {interpMode=RCIM_Cubic, tangentMode=RCTM_User, time=0, value=0.5}
        {arriveTangent=1.5396, interpMode=RCIM_Cubic, leaveTangent=1.5396, tangentMode=RCTM_User, time=1, value=2}
       ]}}})
   /SetVariables(Particles.SubImageIndex=&FloatFromCurve(FloatCurve=DataInterface() {propertyValues {Curve {defaultValue=5.1927, keys [{tangentMode=RCTM_User, time=0, value=0}, {tangentMode=RCTM_User, time=1, value=1}]}}})) {isSetParameters=true, module=SetVariables_#11}
   &DynamicMaterialParameters(
    Index 0 Param 1=&FloatFromCurve(FloatCurve=DataInterface() {propertyValues {Curve {keys [<at -0.0046 1.0113 RCIM_Cubic>,<at 1.0011 0.002>]}}})
    Index 0 Param 4=&FloatFromCurve(FloatCurve=DataInterface() {propertyValues {Curve {keys [<at 0 0 RCIM_Cubic>,<at 1 0.5>]}}})
    Module.Param0WriteEnabled=true
    Module.Param1WriteEnabled=false
    Module.Param2WriteEnabled=false
    Module.Param3WriteEnabled=true
   ) {module=DynamicMatParams}
   &ScaleColor(
    Scale Alpha=&FloatFromCurve(
     FloatCurve=DataInterface() {propertyValues {Curve {keys [
         {interpMode=RCIM_Cubic, tangentMode=RCTM_User, time=0, value=0}
         <at 0.05 1>
         {arriveTangent=-1.0493, interpMode=RCIM_Cubic, leaveTangent=-1.0493, time=0.5, value=1}
         <at 1.0011 0.002>
        ]}}}
     Scale Curve=&FNTimeOfDayScalar(Night Scale=50, Morning Scale=0.7, Evening Scale=0.7)
    )
    ScaleA=true
    ScaleRGB=false
   )
   /SetVariables(
    Particles.WindStreakColor:LinearColor=&MultiplyLinearColor(
     Linear Color=&RandomRangeLinearColor(Minimum=<rgba 0.102 0.102 0.102 1>, Maximum=<rgba 0.624 0.624 0.624 1>)
     Scale Factor=&MakeVector4_FromLinearColor(LinearColor=&MakeLinearColorFromVectorAndFloat(
       "Vector (RGB)"=&MakeVectorFromLinearColorRGB(Color=&FNTimeOfDayColor(
         Night Color=<rgba 0.2123 0.2489 0.27 1>
         Morning Color=<rgba 0.445 0.2629 0.1507 1>
         Day Color=<rgba 0.3626 0.4008 0.4271 1>
         Evening Color=<rgba 0.599 0.2667 0.2028 1>
        ))
       "Float (Alpha)"=&FNTimeOfDayScalar(Night Scale=0.75, Morning Scale=1, Day Scale=0.9, Evening Scale=1.2)
      ))
    )
    Particles.WindStreakSize:Vector2f=&RandomRangeVector2D(Minimum=<3 35>, Maximum=<5 50>)
   ) {isSetParameters=true, module=SetVariables_#12}
   /SetVariables(
    Particles.RingCO=20
    Particles.RingColor:LinearColor=&MultiplyLinearColor(
     Linear Color=<rgba 0.309 0.3956 0.4073 1>
     Scale Factor=&MakeVector4_FromLinearColor(LinearColor=&MakeLinearColorFromVectorAndFloat(
       "Vector (RGB)"=&MakeVectorFromLinearColorRGB(Color=&FNTimeOfDayColor(
         Night Color=<rgba 0.309 0.3956 0.4073 1>
         Morning Color=<rgba 0.445 0.2629 0.1507 1>
         Day Color=<rgba 0.3626 0.4008 0.4271 1>
         Evening Color=<rgba 0.599 0.2667 0.2028 1>
        ))
       "Float (Alpha)"=&FloatFromCurve(
        FloatCurve=DataInterface() {propertyValues {Curve {keys [
            {tangentMode=RCTM_User, time=0, value=1}
            {arriveTangent=-0.0405, interpMode=RCIM_Cubic, leaveTangent=-0.0405, tangentMode=RCTM_User, time=0.7009, value=1}
            {arriveTangent=-8.0089, interpMode=RCIM_Cubic, leaveTangent=-8.0089, tangentMode=RCTM_User, time=1, value=0}
           ]}}}
        Scale Curve=2
       )
      ))
    )
    Particles.RingDP:Vector4f=&MakeVector4(
     X=&FloatFromCurve(FloatCurve=DataInterface() {propertyValues {Curve {keys [
          {arriveTangent=4.8499, interpMode=RCIM_Cubic, leaveTangent=4.8499, tangentMode=RCTM_User, time=0, value=0.5}
          {arriveTangent=0.065, interpMode=RCIM_Cubic, leaveTangent=0.065, tangentMode=RCTM_User, time=1, value=2}
         ]}}})
     Y=0
     Z=0
     W=0
    )
    Particles.RingSize:Vector2f=&Vector2DFromFloat(Value=&Multiply_Float(
      A=5
      B=&FloatFromCurve(
       FloatCurve=DataInterface() {propertyValues {Curve {keys [
           {arriveTangent=2.5949, interpMode=RCIM_Cubic, leaveTangent=2.5949, tangentMode=RCTM_User, time=0, value=0}
           {arriveTangent=0.0224, interpMode=RCIM_Cubic, leaveTangent=0.0224, tangentMode=RCTM_User, time=1, value=1}
          ]}}}
       Scale Curve=25
      )
     ))
   ) {isSetParameters=true, module=SetVariables_#13}
   /SetVariables(Particles.Position=&AddVectorToPosition(Position=<0 0 0>, Vector=<0 0 10>)) {isSetParameters=true, module=SetVariables_#14}
  ]
  rendererClasses [&NiagaraSpriteRendererProperties]
  renderers [
   {rendererClass=&NiagaraSpriteRendererProperties}
   {rendererClass=&NiagaraSpriteRendererProperties, rendererIndex=1}
   {rendererClass=&NiagaraSpriteRendererProperties, rendererIndex=2}
  ]
 }
 SmokeRingStreaks001 {
  emitterUpdateScript [
   &EmitterState(
    Delay First Loop Only=true
    Life Cycle Mode=::NewEnumerator1
    Loop Delay=0.2
    Loop Duration=&RandomRangeFloat(Minimum=0.5, Maximum=0.8)
    Recalculate Duration Each Loop=true
    UseLoopDelay=true
   )
   &SpawnBurst_Instantaneous(
    Spawn Count=&RandomRangeInt(Minimum=15, Maximum=24)
    Use Spawn Probability=false
    UseLoopCountLimit=false
   )
   &SpawnBurst_Instantaneous(
    Spawn Count=&RandomRangeInt(Minimum=15, Maximum=24)
    Spawn Group=1
    Use Spawn Probability=false
    UseLoopCountLimit=false
   ) {module=SpawnBurst_Instantaneous001}
   &SpawnBurst_Instantaneous(Spawn Group=2, Use Spawn Probability=false, UseLoopCountLimit=false) {module=SpawnBurst_Instantaneous002}
  ]
  particleSpawnScript [
   /SetVariables(Particles.VisibilityTag=&SpawnGroup()) {isSetParameters=true, module=SetVariables_#8}
   &InitializeParticle(
    Color=&SelectLinearColorFromArray(
     Array Sampling Mode=::NewEnumerator0
     Direct Array Index=?
     Color Selection Array=DataInterface() {propertyValues {ColorData [<rgba 0.903 0.9271 0.1598 0.05>, <rgba 0.2392 0.9271 0.1932 0.05>, <rgba 0.1336 0.9271 0.8384 0.05>]}}
     Random Seed=0
    )
    Lifetime Max=&MaskFloatBySpawnGroup(
     Masked Float=0.8
     Passthrough Float=&MaskFloatBySpawnGroup(Masked Float=0.4, Passthrough Float=0.8, Spawn Group Mask=1)
    )
    Lifetime Min:NiagaraFloat=&MaskFloatBySpawnGroup(
     Masked Float=0.5
     Passthrough Float=&MaskFloatBySpawnGroup(Masked Float=0.35, Passthrough Float=0.5, Spawn Group Mask=1)
    )
    Lifetime Mode=::NewEnumerator1
    Sprite Rotation Mode=::NewEnumerator2
    Sprite Size Mode=::NewEnumerator3
    Uniform Sprite Size=&RandomRangeFloat(Minimum=15, Maximum=25)
    UsePositionOffset=false
    Write Material Random=true
   )
   &ShapeLocation(
    Cylinder Height=&MaskFloatBySpawnGroup(Masked Float=5, Passthrough Float=0)
    Cylinder Height Midpoint=0
    Cylinder Radius=&MaskFloatBySpawnGroup(
     Masked Float=25
     Passthrough Float=&MaskFloatBySpawnGroup(Masked Float=0.5, Passthrough Float=0, Spawn Group Mask=1)
    )
    MaskSpawnGroup=false
    Override Seed=false
    Shape Primitive=::NewEnumerator1
    Surface Only=true
    Use Endcaps In Surface Only Mode=false
    UseOwnerScale=true
   )
   &AddVelocity(Velocity=&MakeVector(
     X=0
     Y=0
     Z=&UniformRangedFloat(
      Minimum=0
      Maximum=&MaskFloatBySpawnGroup(
       Masked Float=0
       Passthrough Float=&MaskFloatBySpawnGroup(Masked Float=50, Passthrough Float=0, Spawn Group Mask=1)
      )
     )
    )) {module=AddVelocity001}
   &AddVelocity(
    Module.Use Falloff Distance=false
    Override Seed=false
    Velocity Mode=::NewEnumerator1
    Velocity Speed=&RandomRangeFloat(
     Minimum=&MaskFloatBySpawnGroup(
      Masked Float=125
      Passthrough Float=&MaskFloatBySpawnGroup(Masked Float=150, Passthrough Float=0, Spawn Group Mask=1)
     )
     Maximum=&MaskFloatBySpawnGroup(
      Masked Float=200
      Passthrough Float=&MaskFloatBySpawnGroup(Masked Float=150, Passthrough Float=0, Spawn Group Mask=1)
     )
    )
   )
   &CameraOffset(Camera Offset Amount=5)
   /SetVariables(Particles.WindStreakDP:Vector4f=&MakeVector4(
     X=&RandomRangeFloat(Minimum=0, Maximum=1)
     Y=&RandomRangeFloat(Minimum=1, Maximum=2)
     Z=1
     W=1
    )) {isSetParameters=true, module=SetVariables_#9}
   /SetVariables(Particles.RingAlignment:Vector3f=<0 1 0>, Particles.RingFacing:Vector3f=<0 0 1>) {isSetParameters=true, module=SetVariables_#10}
  ]
  particleUpdateScript [
   &ParticleState()
   &GravityForce(Gravity=&MakeVector(
     X=0
     Y=0
     Z=&FloatFromCurve(
      FloatCurve=DataInterface() {propertyValues {Curve {keys [<at 0 0 RCIM_Cubic>,<at 0.4 1 RCIM_Cubic>]}}}
      Scale Curve=&MaskFloatBySpawnGroup(Masked Float=&RandomRangeFloat(Minimum=50, Maximum=125), Passthrough Float=0)
     )
    ))
   &Drag(
    Drag=&MaskFloatBySpawnGroup(Masked Float=10, Passthrough Float=1.5)
    Use Linear Drag=true
    Use Rotational Drag=false
   )
   &SolveForcesAndVelocity(Clamp Velocity=false, Limit Acceleration=false)
   &ScaleSpriteSize(Uniform Curve Sprite Scale=DataInterface() {propertyValues {Curve {keys [
        {interpMode=RCIM_Cubic, tangentMode=RCTM_User, time=0, value=0.5}
        {arriveTangent=1.5396, interpMode=RCIM_Cubic, leaveTangent=1.5396, tangentMode=RCTM_User, time=1, value=2}
       ]}}})
   /SetVariables(Particles.SubImageIndex=&FloatFromCurve(FloatCurve=DataInterface() {propertyValues {Curve {defaultValue=5.1927, keys [{tangentMode=RCTM_User, time=0, value=0}, {tangentMode=RCTM_User, time=1, value=1}]}}})) {isSetParameters=true, module=SetVariables_#11}
   &DynamicMaterialParameters(
    Index 0 Param 1=&FloatFromCurve(FloatCurve=DataInterface() {propertyValues {Curve {keys [<at -0.0046 1.0113 RCIM_Cubic>,<at 1.0011 0.002>]}}})
    Index 0 Param 4=&FloatFromCurve(FloatCurve=DataInterface() {propertyValues {Curve {keys [<at 0 0 RCIM_Cubic>,<at 1 0.5>]}}})
    Module.Param0WriteEnabled=true
    Module.Param1WriteEnabled=false
    Module.Param2WriteEnabled=false
    Module.Param3WriteEnabled=true
   ) {module=DynamicMatParams}
   &ScaleColor(
    Scale Alpha=&FloatFromCurve(
     FloatCurve=DataInterface() {propertyValues {Curve {keys [
         {interpMode=RCIM_Cubic, tangentMode=RCTM_User, time=0, value=0}
         <at 0.05 1>
         {arriveTangent=-1.0493, interpMode=RCIM_Cubic, leaveTangent=-1.0493, time=0.5, value=1}
         <at 1.0011 0.002>
        ]}}}
     Scale Curve=&FNTimeOfDayScalar(Night Scale=50, Morning Scale=0.7, Evening Scale=0.7)
    )
    ScaleA=true
    ScaleRGB=false
   )
   /SetVariables(
    Particles.WindStreakColor:LinearColor=&MultiplyLinearColor(
     Linear Color=&RandomRangeLinearColor(Minimum=<rgba 0.102 0.102 0.102 1>, Maximum=<rgba 0.624 0.624 0.624 1>)
     Scale Factor=&MakeVector4_FromLinearColor(LinearColor=&MakeLinearColorFromVectorAndFloat(
       "Vector (RGB)"=&MakeVectorFromLinearColorRGB(Color=&FNTimeOfDayColor(
         Night Color=<rgba 0.2123 0.2489 0.27 1>
         Morning Color=<rgba 0.445 0.2629 0.1507 1>
         Day Color=<rgba 0.3626 0.4008 0.4271 1>
         Evening Color=<rgba 0.599 0.2667 0.2028 1>
        ))
       "Float (Alpha)"=&FNTimeOfDayScalar(Night Scale=0.75, Morning Scale=1, Day Scale=0.9, Evening Scale=1.2)
      ))
    )
    Particles.WindStreakSize:Vector2f=&RandomRangeVector2D(Minimum=<3 35>, Maximum=<5 50>)
   ) {isSetParameters=true, module=SetVariables_#12}
   /SetVariables(
    Particles.RingCO=20
    Particles.RingColor:LinearColor=&MultiplyLinearColor(
     Linear Color=<rgba 0.309 0.3956 0.4073 1>
     Scale Factor=&MakeVector4_FromLinearColor(LinearColor=&MakeLinearColorFromVectorAndFloat(
       "Vector (RGB)"=&MakeVectorFromLinearColorRGB(Color=&FNTimeOfDayColor(
         Night Color=<rgba 0.309 0.3956 0.4073 1>
         Morning Color=<rgba 0.445 0.2629 0.1507 1>
         Day Color=<rgba 0.3626 0.4008 0.4271 1>
         Evening Color=<rgba 0.599 0.2667 0.2028 1>
        ))
       "Float (Alpha)"=&FloatFromCurve(
        FloatCurve=DataInterface() {propertyValues {Curve {keys [
            {tangentMode=RCTM_User, time=0, value=1}
            {arriveTangent=-0.0405, interpMode=RCIM_Cubic, leaveTangent=-0.0405, tangentMode=RCTM_User, time=0.7009, value=1}
            {arriveTangent=-8.0089, interpMode=RCIM_Cubic, leaveTangent=-8.0089, tangentMode=RCTM_User, time=1, value=0}
           ]}}}
        Scale Curve=2
       )
      ))
    )
    Particles.RingDP:Vector4f=&MakeVector4(
     X=&FloatFromCurve(FloatCurve=DataInterface() {propertyValues {Curve {keys [
          {arriveTangent=4.8499, interpMode=RCIM_Cubic, leaveTangent=4.8499, tangentMode=RCTM_User, time=0, value=0.5}
          {arriveTangent=0.065, interpMode=RCIM_Cubic, leaveTangent=0.065, tangentMode=RCTM_User, time=1, value=2}
         ]}}})
     Y=0
     Z=0
     W=0
    )
    Particles.RingSize:Vector2f=&Vector2DFromFloat(Value=&Multiply_Float(
      A=5
      B=&FloatFromCurve(
       FloatCurve=DataInterface() {propertyValues {Curve {keys [
           {arriveTangent=2.5949, interpMode=RCIM_Cubic, leaveTangent=2.5949, tangentMode=RCTM_User, time=0, value=0}
           {arriveTangent=0.0224, interpMode=RCIM_Cubic, leaveTangent=0.0224, tangentMode=RCTM_User, time=1, value=1}
          ]}}}
       Scale Curve=25
      )
     ))
   ) {isSetParameters=true, module=SetVariables_#13}
   /SetVariables(Particles.Position=&AddVectorToPosition(Position=<0 0 0>, Vector=<0 0 10>)) {isSetParameters=true, module=SetVariables_#14}
  ]
  rendererClasses [&NiagaraSpriteRendererProperties]
  renderers [{rendererClass=&NiagaraSpriteRendererProperties}]
 }
}
summary {
 systemName=NS_Emote_BuffCut_Prop_Hover
 emitters [
  {bEnabled=false, emitterName=SmokeRingStreaks, rendererClasses [&NiagaraSpriteRendererProperties]}
  {bEnabled=false, emitterName=Rainbow_bokeh, rendererClasses [&NiagaraSpriteRendererProperties]}
  {
   bEnabled=false
   emitterName=Fog_system_Condensed
   rendererClasses [
    &NiagaraSpriteRendererProperties
    &NiagaraLightRendererProperties
    &NiagaraMeshRendererProperties
   ]
  }
  {emitterName=SmokeRingStreaks001, rendererClasses [&NiagaraSpriteRendererProperties]}
  {emitterName=Rainbow_bokeh001, rendererClasses [&NiagaraSpriteRendererProperties]}
  {
   emitterName=Fog_system_Condensed001
   rendererClasses [
    &NiagaraSpriteRendererProperties
    &NiagaraLightRendererProperties
    &NiagaraMeshRendererProperties
   ]
  }
 ]
 userVariables [
  User.bIsFrontEnd {type=NiagaraBool}
  User.bIsFrontEndPreview {type=NiagaraBool}
  User.OwnerLocalVelocity {type=Vector3f, description="", defaultValue {struct=Vector3f, value=<0 0 0>}}
  User.PickaxeReactive {type=NiagaraFloat, defaultValue=1, description=""}
 ]
}
