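# Niagara system for the BuffCut cosmetic companion emote that shows short-lived soundwave pulses from both sides of a skeletal mesh.
# Mirrored RIGHT and left002 emitters burst mesh and sprite particles at offset bone/socket positions, with left002 adding a spawn rate.
# Particles move outward under early heavy drag while growing and fading, with color and alpha adjusted by time of day.
# Disabled left003 and left004 variants instead resample position every update and use warmer colors.
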NiagaraSystem /CosmeticCompanions/FX/Builtins/BuffCut/Emote/NS_Emote_BuffCut_soundwaves
dependencies {
 usedDataInterfaces [&NiagaraDataInterfaceCurve, &NiagaraDataInterfaceColorCurve, &NiagaraDataInterfaceSkeletalMesh]
 usedDynamicInputs [&MaskFloatBySpawnGroup]
 usedModules [
  &self:SystemSpawnScript.SystemScriptSource.SystemScriptGraph.NiagaraNodeAssignment_0.SetVariables_#1
  &SystemState
  &Timeline
 ]
}
emitters {
 RIGHT {
  emitterUpdateScript [
   &EmitterState(Life Cycle Mode=::NewEnumerator1, Loop Behavior=::NewEnumerator1, Loop Duration=0.05, Scalability Mode=::NewEnumerator1, UseLoopDelay=false)
   &SpawnBurst_Instantaneous(Spawn Count=2, Use Spawn Probability=false, UseLoopCountLimit=false)
   &SpawnBurst_Instantaneous(Spawn Count=2, Spawn Time=0.02, Use Spawn Probability=false, UseLoopCountLimit=false) {module=SpawnBurst_Instantaneous003}
   &SpawnBurst_Instantaneous(Spawn Count=2, Spawn Time=0.01, Use Spawn Probability=false, UseLoopCountLimit=false) {module=SpawnBurst_Instantaneous001}
  ]
  particleSpawnScript [
   &InitializeParticle(
    Color=&FNTimeOfDayColor(Night Color=<rgba 0.4 0.4 0.4 1>, Morning Color=<rgba 0.65 0.65 0.65 1>, Day Color=<rgba 1 1 1 1>, Evening Color=<rgba 0.65 0.65 0.65 1>)
    Lifetime=0.25
    Mesh Scale=<0.32 0.32 0.32>
    Mesh Scale Mode=::NewEnumerator4
    Position Offset=<5 0 -2.5>
    Sprite Size Mode=::NewEnumerator3
    Uniform Sprite Size=25
    UsePositionOffset=false
    Write Material Random=true
   )
   &CameraOffset()
   &InitialMeshOrientation(
    Mesh Axis To Orient=<0 0 1>
    Mesh Orientation Mode=::NewEnumerator2
    Orientation Coordinate Space=::Local
    Orientation Vector=<0 0 5.01>
    Rotation=<0.15 0 1>
    Rotation Coordinate Space=::NewEnumerator2
    Use Rotation Vector=true
   ) {module=InitialMeshOrientation001}
   &CameraOffset(Camera Offset Amount=2) {enabled=false, module=CameraOffset001}
   &SkeletalMeshLocation(
    MaskSpawnGroup=false
    Mesh Sampling Type=::NewEnumerator5
    Position Sampling=::NewEnumerator1
    Sampled Position Offset=<-13.55 25 16>
    Sampled Position Offset Coordinate Space:ENiagaraSkelMeshTransforms=::NewEnumerator2
    Skeletal Mesh=$System.Skeletal Mesh
    Socket Sampling Mode=::NewEnumerator5
   )
   &ApplyInitialForces()
   &AddVelocity(Velocity=<-15 15 0>)
   &SpriteFacingAndAlignment(
    Facing Coordinate Space=::Local
    Sprite Alignment=<34.095 28.3634 12.5594>
    Sprite Facing=<-0.6571 1 0>
    WriteAlignment=false
    WriteFacing=true
   )
  ]
  particleUpdateScript [
   &ParticleState()
   &ScaleMeshSize(Scale Factor=&VectorFromFloat(Value=&Add_Float(
      A=0.6
      B=&FloatFromCurve(
       FloatCurve=DataInterface() {propertyValues {Curve {keys [
           {
            arriveTangent=1.0578
            arriveTangentWeight=0.2235
            interpMode=RCIM_Cubic
            leaveTangent=1.0578
            leaveTangentWeight=0.2235
            tangentMode=RCTM_User
            time=0
            value=0
           }
           {
            arriveTangent=5.2964
            arriveTangentWeight=0.9219
            interpMode=RCIM_Cubic
            leaveTangent=5.2964
            leaveTangentWeight=0.9219
            tangentMode=RCTM_User
            time=0.5
            value=2.2
           }
          ]}}}
       Scale Curve=0.5
      )
     )))
   &ScaleSpriteSize(Uniform Curve Sprite Scale=DataInterface() {propertyValues {Curve {keys [
        {
         arriveTangent=1.0578
         arriveTangentWeight=0.2235
         interpMode=RCIM_Cubic
         leaveTangent=1.0578
         leaveTangentWeight=0.2235
         tangentMode=RCTM_User
         time=0.05
         value=0
        }
        {
         arriveTangent=5.2964
         arriveTangentWeight=0.9219
         interpMode=RCIM_Cubic
         leaveTangent=5.2964
         leaveTangentWeight=0.9219
         tangentMode=RCTM_User
         time=0.85
         value=2
        }
       ]}}})
   &ScaleColor(
    Scale Alpha=&FloatFromCurve(
     FloatCurve=DataInterface() {propertyValues {Curve {keys [
         {arriveTangent=26.1682, interpMode=RCIM_Cubic, leaveTangent=26.1682, tangentMode=RCTM_User, time=0, value=0.3}
         {arriveTangent=-3.0679, interpMode=RCIM_Cubic, leaveTangent=-3.0679, tangentMode=RCTM_User, time=0.1, value=1}
         {tangentMode=RCTM_User, time=1, value=0}
        ]}}}
     Scale Curve=&FNTimeOfDayScalar(Night Scale=2.7, Morning Scale=2, Day Scale=1.5, Evening Scale=1.85)
    )
    Scale RGB=&Multiply_VectorByFloat(
     Vector=&MakeVectorFromLinearColorRGB(Color=&ColorFromCurve(ColorCurve=DataInterface() {propertyValues {
         AlphaCurve {keys []}
         BlueCurve {keys [<at 0 0.4969 RCIM_Cubic>,<at 1 0.3156 RCIM_Cubic>]}
         GreenCurve {keys [<at 0 0.4125 RCIM_Cubic>,<at 1 0.3156 RCIM_Cubic>]}
         RedCurve {keys [<at 0 0.0976 RCIM_Cubic>,<at 1 0.3156 RCIM_Cubic>]}
        }}))
     Float=1
    )
    ScaleA=true
    ScaleRGB=true
   )
   &DynamicMaterialParameters(Module.Index0Param0Write=true, Module.Index0Param1Write=true, Module.Index0Param2Write=true, Module.Index0Param3Write=true)
   &Drag(
    Drag=&FloatFromCurve(
     FloatCurve=DataInterface() {propertyValues {Curve {
        defaultValue=1
        keys [{interpMode=RCIM_Cubic, tangentMode=RCTM_User, time=0, value=1}, {interpMode=RCIM_Cubic, tangentMode=RCTM_User, time=0.2, value=0}]
       }}}
     Scale Curve=55
    )
    Use Linear Drag=true
    Use Rotational Drag=true
   )
   &SolveForcesAndVelocity(Clamp Velocity=false, Limit Acceleration=false)
   /SetVariables(Particles.GlowColor=&MakeLinearColorFromVectorAndFloat(
     "Vector (RGB)"=&Multiply_VectorByFloat(
      Vector=&MakeVectorFromLinearColorRGB(Color=&ColorFromCurve(
        ColorCurve=DataInterface() {propertyValues {
          AlphaCurve {keys []}
          BlueCurve {keys [<at 0.2415 0.7>, {arriveTangentWeight=0.1167, interpMode=RCIM_Cubic, leaveTangentWeight=0.8633, tangentMode=RCTM_User, time=1, value=0.4969}]}
          GreenCurve {keys [
            <at 0.2415 0.4138>
            {arriveTangentWeight=0.1167, interpMode=RCIM_Cubic, leaveTangentWeight=0.8633, tangentMode=RCTM_User, time=1, value=0.4125}
           ]}
          RedCurve {keys [
            <at 0.2415 0.3566>
            {arriveTangentWeight=0.1167, interpMode=RCIM_Cubic, leaveTangentWeight=0.8633, tangentMode=RCTM_User, time=1, value=0.0976}
           ]}
         }}
        CurveIndex=$Emitter.NormalizedLoopAge
       ))
      Float=3
     )
     "Float (Alpha)"=&FloatFromCurve(
      FloatCurve=DataInterface() {propertyValues {Curve {keys [
          {arriveTangent=-1.1111, interpMode=RCIM_Cubic, time=0, value=0}
          {arriveTangent=-0.1741, interpMode=RCIM_Cubic, leaveTangent=-0.1741, tangentMode=RCTM_User, time=0.1, value=1}
          {arriveTangent=-0.9656, interpMode=RCIM_Cubic, leaveTangent=-0.9656, tangentMode=RCTM_User, time=0.6, value=0}
         ]}}}
      Scale Curve=&FNTimeOfDayScalar(Night Scale=0.6, Morning Scale=0.7, Day Scale=0.85, Evening Scale=0.7)
     )
    )) {isSetParameters=true, module=SetVariables_#2}
  ]
  rendererClasses [&NiagaraMeshRendererProperties, &NiagaraSpriteRendererProperties]
  renderers [
   {rendererClass=&NiagaraMeshRendererProperties, rendererIndex=0}
   {rendererClass=&NiagaraMeshRendererProperties, rendererIndex=1}
   {rendererClass=&NiagaraSpriteRendererProperties, rendererIndex=2}
   {rendererClass=&NiagaraMeshRendererProperties, rendererIndex=3}
   {rendererClass=&NiagaraMeshRendererProperties, rendererIndex=4}
  ]
 }
 left002 {
  emitterUpdateScript [
   &EmitterState(Life Cycle Mode=::NewEnumerator1, Loop Behavior=::NewEnumerator1, Loop Duration=0.05, Scalability Mode=::NewEnumerator1, UseLoopDelay=false)
   &SpawnRate(SpawnRate=50, Use Spawn Probability=false)
   &SpawnBurst_Instantaneous(Spawn Count=2, Use Spawn Probability=false, UseLoopCountLimit=false)
   &SpawnBurst_Instantaneous(Spawn Count=2, Spawn Time=0.025, Use Spawn Probability=false, UseLoopCountLimit=false) {module=SpawnBurst_Instantaneous003}
   &SpawnBurst_Instantaneous(Spawn Count=2, Spawn Time=0.015, Use Spawn Probability=false, UseLoopCountLimit=false) {module=SpawnBurst_Instantaneous001}
  ]
  particleSpawnScript [
   &InitializeParticle(
    Color=&FNTimeOfDayColor(Night Color=<rgba 0.4 0.4 0.4 1>, Morning Color=<rgba 0.65 0.65 0.65 1>, Day Color=<rgba 1 1 1 1>, Evening Color=<rgba 0.65 0.65 0.65 1>)
    Lifetime=0.25
    Mesh Scale=<0.32 0.32 0.32>
    Mesh Scale Mode=::NewEnumerator4
    Position Offset=<5 0 -2.5>
    Sprite Size Mode=::NewEnumerator3
    Uniform Sprite Size=25
    UsePositionOffset=false
    Write Material Random=true
   )
   &CameraOffset()
   &InitialMeshOrientation(
    Mesh Axis To Orient=<0 0 1>
    Mesh Orientation Mode=::NewEnumerator2
    Orientation Coordinate Space=::Local
    Orientation Vector=<0 0 1>
    Rotation=<0.25 0 1>
    Rotation Coordinate Space=::NewEnumerator2
    Use Rotation Vector=true
   ) {module=InitialMeshOrientation001}
   &CameraOffset(Camera Offset Amount=2) {enabled=false, module=CameraOffset001}
   &ApplyInitialForces()
   &SkeletalMeshLocation(
    Bone / Socket Index=1
    MaskSpawnGroup=false
    Mesh Sampling Type=::NewEnumerator5
    Sampled Position Offset=<13.55 24.85 16.7>
    Sampled Position Offset Coordinate Space:ENiagaraSkelMeshTransforms=::NewEnumerator2
    Skeletal Mesh=$System.Skeletal Mesh
    Socket Sampling Mode=::NewEnumerator5
   )
   &AddVelocity(Velocity=<15 15 0>)
   &SpriteFacingAndAlignment(
    Facing Coordinate Space=::Local
    Sprite Alignment=<34.095 28.3634 12.5594>
    Sprite Facing=<0.5929 1 0>
    WriteAlignment=false
    WriteFacing=true
   )
  ]
  particleUpdateScript [
   &ParticleState()
   &ScaleMeshSize(Scale Factor=&VectorFromFloat(Value=&Add_Float(
      A=0.6
      B=&FloatFromCurve(
       FloatCurve=DataInterface() {propertyValues {Curve {keys [
           {
            arriveTangent=1.0578
            arriveTangentWeight=0.2235
            interpMode=RCIM_Cubic
            leaveTangent=1.0578
            leaveTangentWeight=0.2235
            tangentMode=RCTM_User
            time=0
            value=0
           }
           {
            arriveTangent=5.2964
            arriveTangentWeight=0.9219
            interpMode=RCIM_Cubic
            leaveTangent=5.2964
            leaveTangentWeight=0.9219
            tangentMode=RCTM_User
            time=0.5
            value=2.2
           }
          ]}}}
       Scale Curve=0.5
      )
     )))
   &ScaleSpriteSize(Uniform Curve Sprite Scale=DataInterface() {propertyValues {Curve {keys [
        {
         arriveTangent=1.0578
         arriveTangentWeight=0.2235
         interpMode=RCIM_Cubic
         leaveTangent=1.0578
         leaveTangentWeight=0.2235
         tangentMode=RCTM_User
         time=0.05
         value=0
        }
        {
         arriveTangent=5.2964
         arriveTangentWeight=0.9219
         interpMode=RCIM_Cubic
         leaveTangent=5.2964
         leaveTangentWeight=0.9219
         tangentMode=RCTM_User
         time=0.85
         value=2
        }
       ]}}})
   &ScaleColor(
    Scale Alpha=&FloatFromCurve(
     FloatCurve=DataInterface() {propertyValues {Curve {keys [
         {arriveTangent=26.1682, interpMode=RCIM_Cubic, leaveTangent=26.1682, tangentMode=RCTM_User, time=0, value=0.3}
         {arriveTangent=-3.3297, interpMode=RCIM_Cubic, leaveTangent=-3.3297, tangentMode=RCTM_User, time=0.1, value=1}
         {tangentMode=RCTM_User, time=1, value=0}
        ]}}}
     Scale Curve=&FNTimeOfDayScalar(Night Scale=3.7, Morning Scale=3.2, Day Scale=2.25, Evening Scale=3)
    )
    Scale RGB=&Multiply_VectorByFloat(
     Vector=&MakeVectorFromLinearColorRGB(Color=&ColorFromCurve(ColorCurve=DataInterface() {propertyValues {
         AlphaCurve {keys []}
         BlueCurve {keys [<at 0 0.4969 RCIM_Cubic>,<at 1 0.3156 RCIM_Cubic>]}
         GreenCurve {keys [<at 0 0.4125 RCIM_Cubic>,<at 1 0.3156 RCIM_Cubic>]}
         RedCurve {keys [<at 0 0.0976 RCIM_Cubic>,<at 1 0.3156 RCIM_Cubic>]}
        }}))
     Float=1
    )
    ScaleA=true
    ScaleRGB=true
   )
   &DynamicMaterialParameters(Module.Index0Param0Write=true, Module.Index0Param1Write=true, Module.Index0Param2Write=true, Module.Index0Param3Write=true)
   &Drag(
    Drag=&FloatFromCurve(
     FloatCurve=DataInterface() {propertyValues {Curve {
        defaultValue=1
        keys [{interpMode=RCIM_Cubic, tangentMode=RCTM_User, time=0, value=1}, {interpMode=RCIM_Cubic, tangentMode=RCTM_User, time=0.2, value=0}]
       }}}
     Scale Curve=55
    )
    Use Linear Drag=true
    Use Rotational Drag=true
   )
   &SolveForcesAndVelocity(Clamp Velocity=false, Limit Acceleration=false)
   /SetVariables(Particles.GlowColor=&MakeLinearColorFromVectorAndFloat(
     "Vector (RGB)"=&Multiply_VectorByFloat(
      Vector=&MakeVectorFromLinearColorRGB(Color=&ColorFromCurve(
        ColorCurve=DataInterface() {propertyValues {
          AlphaCurve {keys []}
          BlueCurve {keys [<at 0.2415 0.7>, {arriveTangentWeight=0.1167, interpMode=RCIM_Cubic, leaveTangentWeight=0.8633, tangentMode=RCTM_User, time=1, value=0.4969}]}
          GreenCurve {keys [
            <at 0.2415 0.4138>
            {arriveTangentWeight=0.1167, interpMode=RCIM_Cubic, leaveTangentWeight=0.8633, tangentMode=RCTM_User, time=1, value=0.4125}
           ]}
          RedCurve {keys [
            <at 0.2415 0.3566>
            {arriveTangentWeight=0.1167, interpMode=RCIM_Cubic, leaveTangentWeight=0.8633, tangentMode=RCTM_User, time=1, value=0.0976}
           ]}
         }}
        CurveIndex=$Emitter.NormalizedLoopAge
       ))
      Float=3
     )
     "Float (Alpha)"=&FloatFromCurve(
      FloatCurve=DataInterface() {propertyValues {Curve {keys [
          {arriveTangent=-1.1111, interpMode=RCIM_Cubic, time=0, value=0}
          {arriveTangent=-0.1741, interpMode=RCIM_Cubic, leaveTangent=-0.1741, tangentMode=RCTM_User, time=0.1, value=1}
          {arriveTangent=-0.9656, interpMode=RCIM_Cubic, leaveTangent=-0.9656, tangentMode=RCTM_User, time=0.6, value=0}
         ]}}}
      Scale Curve=&FNTimeOfDayScalar(Night Scale=0.6, Morning Scale=0.7, Day Scale=0.85, Evening Scale=0.7)
     )
    )) {isSetParameters=true, module=SetVariables_#2}
  ]
  rendererClasses [&NiagaraMeshRendererProperties, &NiagaraSpriteRendererProperties]
  renderers [
   {rendererClass=&NiagaraMeshRendererProperties, rendererIndex=0}
   {rendererClass=&NiagaraMeshRendererProperties, rendererIndex=1}
   {rendererClass=&NiagaraSpriteRendererProperties, rendererIndex=2}
   {rendererClass=&NiagaraMeshRendererProperties, rendererIndex=3}
   {rendererClass=&NiagaraMeshRendererProperties, rendererIndex=4}
  ]
 }
 left003 {
  bEnabled=false
  emitterUpdateScript [
   &EmitterState(Life Cycle Mode=::NewEnumerator1, Loop Behavior=::NewEnumerator1, Loop Duration=0.05, Scalability Mode=::NewEnumerator1, UseLoopDelay=false)
   &SpawnBurst_Instantaneous(Spawn Count=2, Use Spawn Probability=false, UseLoopCountLimit=false)
   &SpawnBurst_Instantaneous(Spawn Count=2, Spawn Time=0.02, Use Spawn Probability=false, UseLoopCountLimit=false) {module=SpawnBurst_Instantaneous003}
   &SpawnBurst_Instantaneous(Spawn Count=2, Spawn Time=0.01, Use Spawn Probability=false, UseLoopCountLimit=false) {module=SpawnBurst_Instantaneous001}
  ]
  particleSpawnScript [
   &InitializeParticle(
    Color=&FNTimeOfDayColor(Night Color=<rgba 0.4 0.4 0.4 1>, Morning Color=<rgba 0.65 0.65 0.65 1>, Day Color=<rgba 1 1 1 1>, Evening Color=<rgba 0.65 0.65 0.65 1>)
    Lifetime=0.25
    Mesh Scale=<0.35 0.35 0.35>
    Mesh Scale Mode=::NewEnumerator4
    Position Offset=<5 0 -2.5>
    Sprite Size Mode=::NewEnumerator3
    Uniform Sprite Size=25
    UsePositionOffset=false
    Write Material Random=true
   )
   &CameraOffset()
   &InitialMeshOrientation(
    Mesh Axis To Orient=<0 0 1>
    Mesh Orientation Mode=::NewEnumerator2
    Orientation Coordinate Space=::Local
    Orientation Vector=<0 0 1>
    Rotation=<0.25 0 1>
    Rotation Coordinate Space=::NewEnumerator2
    Use Rotation Vector=false
   ) {module=InitialMeshOrientation001}
   &CameraOffset(Camera Offset Amount=2) {enabled=false, module=CameraOffset001}
   &ApplyInitialForces()
  ]
  particleUpdateScript [
   &ParticleState()
   &SkeletalMeshLocation(
    MaskSpawnGroup=false
    Mesh Sampling Type=::NewEnumerator5
    Position Sampling=::NewEnumerator1
    Sampled Position Offset=<-12 25 16>
    Sampled Position Offset Coordinate Space:ENiagaraSkelMeshTransforms=::NewEnumerator2
    Skeletal Mesh=$System.Skeletal Mesh
    Socket Sampling Mode=::NewEnumerator5
   )
   &ScaleMeshSize(Scale Factor=&VectorFromFloat(Value=&Add_Float(
      A=0.6
      B=&FloatFromCurve(
       FloatCurve=DataInterface() {propertyValues {Curve {keys [
           {
            arriveTangent=1.0578
            arriveTangentWeight=0.2235
            interpMode=RCIM_Cubic
            leaveTangent=1.0578
            leaveTangentWeight=0.2235
            tangentMode=RCTM_User
            time=0
            value=0
           }
           {
            arriveTangent=5.2964
            arriveTangentWeight=0.9219
            interpMode=RCIM_Cubic
            leaveTangent=5.2964
            leaveTangentWeight=0.9219
            tangentMode=RCTM_User
            time=0.5
            value=1
           }
          ]}}}
       Scale Curve=0.5
      )
     )))
   &ScaleSpriteSize(Uniform Curve Sprite Scale=DataInterface() {propertyValues {Curve {keys [
        {
         arriveTangent=1.0578
         arriveTangentWeight=0.2235
         interpMode=RCIM_Cubic
         leaveTangent=1.0578
         leaveTangentWeight=0.2235
         tangentMode=RCTM_User
         time=0.05
         value=0
        }
        {
         arriveTangent=5.2964
         arriveTangentWeight=0.9219
         interpMode=RCIM_Cubic
         leaveTangent=5.2964
         leaveTangentWeight=0.9219
         tangentMode=RCTM_User
         time=1
         value=2
        }
       ]}}})
   &ScaleColor(
    Scale Alpha=&FloatFromCurve(
     FloatCurve=DataInterface() {propertyValues {Curve {keys [
         {arriveTangent=26.1682, interpMode=RCIM_Cubic, leaveTangent=26.1682, tangentMode=RCTM_User, time=0, value=1}
         {arriveTangent=-3.7386, interpMode=RCIM_Cubic, leaveTangent=-3.7386, tangentMode=RCTM_User, time=0.1, value=1}
         {tangentMode=RCTM_User, time=1, value=0}
        ]}}}
     Scale Curve=&FNTimeOfDayScalar(Night Scale=0.5, Morning Scale=0.5, Day Scale=0.35, Evening Scale=0.5)
    )
    Scale RGB=&Multiply_VectorByFloat(
     Vector=&MakeVectorFromLinearColorRGB(Color=&ColorFromCurve(
       ColorCurve=DataInterface() {propertyValues {
         AlphaCurve {keys []}
         BlueCurve {keys [<at 0 1.8319 RCIM_Cubic>,<at 0.7225 0.2 RCIM_Cubic>]}
         GreenCurve {keys [<at 0 9.8759 RCIM_Cubic>,<at 0.7225 0.2 RCIM_Cubic>]}
         RedCurve {keys [<at 0 20 RCIM_Cubic>,<at 0.7225 0.2 RCIM_Cubic>]}
        }}
       CurveIndex=$Emitter.NormalizedLoopAge
      ))
     Float=5
    )
    ScaleA=true
    ScaleRGB=true
   )
   &DynamicMaterialParameters(Module.Index0Param0Write=true, Module.Index0Param1Write=true, Module.Index0Param2Write=true, Module.Index0Param3Write=true)
   &AccelerationForce(
    Acceleration=&MaskVectorBySpawnGroup(Masked Vector=<-200 500 0>, Passthrough Vector=<0 0 1>)
    Coordinate Space=::Local
   )
   &SolveForcesAndVelocity(Clamp Velocity=false, Limit Acceleration=false)
   &AlignSpriteToMeshOrientation(
    Mesh Orientation Relative Sprite Alignment Vector=<1 1 0>
    Mesh Orientation Relative Sprite Facing Vector:Vector3f=<1 1 0>
    SetAlignment=true
    SetFacing=true
   ) {enabled=false}
   /SetVariables(Particles.GlowColor=&MakeLinearColorFromVectorAndFloat(
     "Vector (RGB)"=&Multiply_VectorByFloat(
      Vector=&MakeVectorFromLinearColorRGB(Color=&ColorFromCurve(
        ColorCurve=DataInterface() {propertyValues {
          AlphaCurve {keys []}
          BlueCurve {keys [
            <at 0.2415 1>
            {tangentMode=RCTM_User, time=0.4905, value=0.0004}
            {arriveTangentWeight=0.1167, interpMode=RCIM_Cubic, leaveTangentWeight=0.8633, tangentMode=RCTM_User, time=1, value=0}
           ]}
          GreenCurve {keys [
            <at 0.2415 0.3997>
            {tangentMode=RCTM_User, time=0.4905, value=0.5703}
            {arriveTangentWeight=0.1167, interpMode=RCIM_Cubic, leaveTangentWeight=0.8633, tangentMode=RCTM_User, time=1, value=0.8929}
           ]}
          RedCurve {keys [
            <at 0.2415 0>
            {tangentMode=RCTM_User, time=0.4905, value=0.186}
            {arriveTangentWeight=0.1167, interpMode=RCIM_Cubic, leaveTangentWeight=0.8633, tangentMode=RCTM_User, time=1, value=1.25}
           ]}
         }}
        CurveIndex=$Emitter.NormalizedLoopAge
       ))
      Float=3
     )
     "Float (Alpha)"=&FloatFromCurve(
      FloatCurve=DataInterface() {propertyValues {Curve {keys [
          {arriveTangent=-1.1111, interpMode=RCIM_Cubic, time=0, value=0}
          {arriveTangent=-0.1741, interpMode=RCIM_Cubic, leaveTangent=-0.1741, tangentMode=RCTM_User, time=0.1, value=1}
          {arriveTangent=-0.9656, interpMode=RCIM_Cubic, leaveTangent=-0.9656, tangentMode=RCTM_User, time=0.6, value=0}
         ]}}}
      Scale Curve=&FNTimeOfDayScalar(Night Scale=0.6, Morning Scale=0.7, Day Scale=0.85, Evening Scale=0.7)
     )
    )) {isSetParameters=true, module=SetVariables_#2}
  ]
  rendererClasses [&NiagaraMeshRendererProperties, &NiagaraSpriteRendererProperties]
  renderers [
   {rendererClass=&NiagaraMeshRendererProperties, rendererIndex=0}
   {rendererClass=&NiagaraMeshRendererProperties, rendererIndex=1}
   {rendererClass=&NiagaraSpriteRendererProperties, rendererIndex=2}
   {rendererClass=&NiagaraMeshRendererProperties, rendererIndex=3}
   {rendererClass=&NiagaraMeshRendererProperties, rendererIndex=4}
  ]
 }
 left004 {
  bEnabled=false
  emitterUpdateScript [
   &EmitterState(Life Cycle Mode=::NewEnumerator1, Loop Behavior=::NewEnumerator1, Loop Duration=0.05, Scalability Mode=::NewEnumerator1, UseLoopDelay=false)
   &SpawnBurst_Instantaneous(Spawn Count=2, Use Spawn Probability=false, UseLoopCountLimit=false)
   &SpawnBurst_Instantaneous(Spawn Count=2, Spawn Time=0.02, Use Spawn Probability=false, UseLoopCountLimit=false) {module=SpawnBurst_Instantaneous003}
   &SpawnBurst_Instantaneous(Spawn Count=2, Spawn Time=0.01, Use Spawn Probability=false, UseLoopCountLimit=false) {module=SpawnBurst_Instantaneous001}
  ]
  particleSpawnScript [
   &InitializeParticle(
    Color=&FNTimeOfDayColor(Night Color=<rgba 0.4 0.4 0.4 1>, Morning Color=<rgba 0.65 0.65 0.65 1>, Day Color=<rgba 1 1 1 1>, Evening Color=<rgba 0.65 0.65 0.65 1>)
    Lifetime=0.25
    Mesh Scale=<0.35 0.35 0.35>
    Mesh Scale Mode=::NewEnumerator4
    Position Offset=<5 0 -2.5>
    Sprite Size Mode=::NewEnumerator3
    Uniform Sprite Size=25
    UsePositionOffset=false
    Write Material Random=true
   )
   &CameraOffset()
   &InitialMeshOrientation(
    Mesh Axis To Orient=<0 0 1>
    Mesh Orientation Mode=::NewEnumerator2
    Orientation Coordinate Space=::Local
    Orientation Vector=<0 0 1>
    Rotation=<0.25 0 1>
    Rotation Coordinate Space=::NewEnumerator2
    Use Rotation Vector=false
   ) {module=InitialMeshOrientation001}
   &CameraOffset(Camera Offset Amount=2) {enabled=false, module=CameraOffset001}
   &ApplyInitialForces()
  ]
  particleUpdateScript [
   &ParticleState()
   &SkeletalMeshLocation(
    Bone / Socket Index=1
    MaskSpawnGroup=false
    Mesh Sampling Type=::NewEnumerator5
    Position Sampling=::NewEnumerator1
    Sampled Position Offset=<13 25 16>
    Sampled Position Offset Coordinate Space:ENiagaraSkelMeshTransforms=::NewEnumerator2
    Skeletal Mesh=$System.Skeletal Mesh
    Socket Sampling Mode=::NewEnumerator5
   )
   &ScaleMeshSize(Scale Factor=&VectorFromFloat(Value=&Add_Float(
      A=0.6
      B=&FloatFromCurve(
       FloatCurve=DataInterface() {propertyValues {Curve {keys [
           {
            arriveTangent=1.0578
            arriveTangentWeight=0.2235
            interpMode=RCIM_Cubic
            leaveTangent=1.0578
            leaveTangentWeight=0.2235
            tangentMode=RCTM_User
            time=0
            value=0
           }
           {
            arriveTangent=5.2964
            arriveTangentWeight=0.9219
            interpMode=RCIM_Cubic
            leaveTangent=5.2964
            leaveTangentWeight=0.9219
            tangentMode=RCTM_User
            time=0.5
            value=1
           }
          ]}}}
       Scale Curve=0.5
      )
     )))
   &ScaleSpriteSize(Uniform Curve Sprite Scale=DataInterface() {propertyValues {Curve {keys [
        {
         arriveTangent=1.0578
         arriveTangentWeight=0.2235
         interpMode=RCIM_Cubic
         leaveTangent=1.0578
         leaveTangentWeight=0.2235
         tangentMode=RCTM_User
         time=0.05
         value=0
        }
        {
         arriveTangent=5.2964
         arriveTangentWeight=0.9219
         interpMode=RCIM_Cubic
         leaveTangent=5.2964
         leaveTangentWeight=0.9219
         tangentMode=RCTM_User
         time=1
         value=2
        }
       ]}}})
   &ScaleColor(
    Scale Alpha=&FloatFromCurve(
     FloatCurve=DataInterface() {propertyValues {Curve {keys [
         {arriveTangent=26.1682, interpMode=RCIM_Cubic, leaveTangent=26.1682, tangentMode=RCTM_User, time=0, value=1}
         {arriveTangent=-3.7386, interpMode=RCIM_Cubic, leaveTangent=-3.7386, tangentMode=RCTM_User, time=0.1, value=1}
         {tangentMode=RCTM_User, time=1, value=0}
        ]}}}
     Scale Curve=&FNTimeOfDayScalar(Night Scale=0.5, Morning Scale=0.5, Day Scale=0.35, Evening Scale=0.5)
    )
    Scale RGB=&Multiply_VectorByFloat(
     Vector=&MakeVectorFromLinearColorRGB(Color=&ColorFromCurve(
       ColorCurve=DataInterface() {propertyValues {
         AlphaCurve {keys []}
         BlueCurve {keys [<at 0 1.8319 RCIM_Cubic>,<at 0.7225 0.2 RCIM_Cubic>]}
         GreenCurve {keys [<at 0 9.8759 RCIM_Cubic>,<at 0.7225 0.2 RCIM_Cubic>]}
         RedCurve {keys [<at 0 20 RCIM_Cubic>,<at 0.7225 0.2 RCIM_Cubic>]}
        }}
       CurveIndex=$Emitter.NormalizedLoopAge
      ))
     Float=5
    )
    ScaleA=true
    ScaleRGB=true
   )
   &DynamicMaterialParameters(Module.Index0Param0Write=true, Module.Index0Param1Write=true, Module.Index0Param2Write=true, Module.Index0Param3Write=true)
   &AccelerationForce(
    Acceleration=&MaskVectorBySpawnGroup(Masked Vector=<200 500 0>, Passthrough Vector=<0 0 1>)
    Coordinate Space=::Local
   )
   &SolveForcesAndVelocity(Clamp Velocity=false, Limit Acceleration=false)
   &AlignSpriteToMeshOrientation(
    Mesh Orientation Relative Sprite Alignment Vector=<1 1 0>
    Mesh Orientation Relative Sprite Facing Vector:Vector3f=<1 1 0>
    SetAlignment=true
    SetFacing=true
   ) {enabled=false}
   /SetVariables(Particles.GlowColor=&MakeLinearColorFromVectorAndFloat(
     "Vector (RGB)"=&Multiply_VectorByFloat(
      Vector=&MakeVectorFromLinearColorRGB(Color=&ColorFromCurve(
        ColorCurve=DataInterface() {propertyValues {
          AlphaCurve {keys []}
          BlueCurve {keys [
            <at 0.2415 1>
            {tangentMode=RCTM_User, time=0.4905, value=0.0004}
            {arriveTangentWeight=0.1167, interpMode=RCIM_Cubic, leaveTangentWeight=0.8633, tangentMode=RCTM_User, time=1, value=0}
           ]}
          GreenCurve {keys [
            <at 0.2415 0.3997>
            {tangentMode=RCTM_User, time=0.4905, value=0.5703}
            {arriveTangentWeight=0.1167, interpMode=RCIM_Cubic, leaveTangentWeight=0.8633, tangentMode=RCTM_User, time=1, value=0.8929}
           ]}
          RedCurve {keys [
            <at 0.2415 0>
            {tangentMode=RCTM_User, time=0.4905, value=0.186}
            {arriveTangentWeight=0.1167, interpMode=RCIM_Cubic, leaveTangentWeight=0.8633, tangentMode=RCTM_User, time=1, value=1.25}
           ]}
         }}
        CurveIndex=$Emitter.NormalizedLoopAge
       ))
      Float=3
     )
     "Float (Alpha)"=&FloatFromCurve(
      FloatCurve=DataInterface() {propertyValues {Curve {keys [
          {arriveTangent=-1.1111, interpMode=RCIM_Cubic, time=0, value=0}
          {arriveTangent=-0.1741, interpMode=RCIM_Cubic, leaveTangent=-0.1741, tangentMode=RCTM_User, time=0.1, value=1}
          {arriveTangent=-0.9656, interpMode=RCIM_Cubic, leaveTangent=-0.9656, tangentMode=RCTM_User, time=0.6, value=0}
         ]}}}
      Scale Curve=&FNTimeOfDayScalar(Night Scale=0.6, Morning Scale=0.7, Day Scale=0.85, Evening Scale=0.7)
     )
    )) {isSetParameters=true, module=SetVariables_#2}
  ]
  rendererClasses [&NiagaraMeshRendererProperties, &NiagaraSpriteRendererProperties]
  renderers [
   {rendererClass=&NiagaraMeshRendererProperties, rendererIndex=0}
   {rendererClass=&NiagaraMeshRendererProperties, rendererIndex=1}
   {rendererClass=&NiagaraSpriteRendererProperties, rendererIndex=2}
   {rendererClass=&NiagaraMeshRendererProperties, rendererIndex=3}
   {rendererClass=&NiagaraMeshRendererProperties, rendererIndex=4}
  ]
 }
}
summary {
 systemName=NS_Emote_BuffCut_soundwaves
 emitters [
  {emitterName=RIGHT, rendererClasses [&NiagaraMeshRendererProperties, &NiagaraSpriteRendererProperties]}
  {emitterName=left002, rendererClasses [&NiagaraMeshRendererProperties, &NiagaraSpriteRendererProperties]}
  {
   bEnabled=false
   emitterName=left003
   rendererClasses [&NiagaraMeshRendererProperties, &NiagaraSpriteRendererProperties]
  }
  {
   bEnabled=false
   emitterName=left004
   rendererClasses [&NiagaraMeshRendererProperties, &NiagaraSpriteRendererProperties]
  }
 ]
 userVariables [
  User.bIsFrontEnd {type=NiagaraBool}
  User.bIsFrontEndPreview {type=NiagaraBool}
  User.PickaxeReactive {type=NiagaraFloat, defaultValue=1, description=""}
 ]
}
